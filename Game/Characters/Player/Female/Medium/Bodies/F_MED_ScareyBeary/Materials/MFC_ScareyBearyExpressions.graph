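# Material function for the ScareyBeary character that switches facial expressions by hiding every eye and mouth mesh variant except the active one.
# VertexColorMask tags variants by value in configurable vertex color channels, and chained threshold Ifs on pose scalars pick one per region by priority, falling back to default.
# Subtracting that mask from the Max of all variants drives World Position Offset, collapsing unselected vertices to the object position.
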MaterialFunction /Game/Characters/Player/Female/Medium/Bodies/F_MED_ScareyBeary/Materials/MFC_ScareyBearyExpressions
graph {nodes [
  n1 = FunctionInput() {inputType=FunctionInput_MaterialAttributes}
  n2 = Constant() {r=1}
  n3 = VertexColor()
  n4 = ChannelMaskParameter(None=@n3) {group=ScareyBearyExpression, parameterName=SBE_EyesChannel, sortPriority=-1}
  n5 = MaterialFunctionCall!(
   Hide Element Toggle=@n2
   VertexColorChannel=@n4.R
   Vertex Value to hide=ScalarParameter() {defaultValue=0.4, group=ScareyBearyExpression, parameterName=R_blink_pose_value}
  ) {materialFunction=/Game/Packages/Fortress_SharedMaterials/Base_Material_Functions/VertexColorMask}
  n6 = MaterialFunctionCall(
   Hide Element Toggle=@n2
   VertexColorChannel=@n4.R
   Vertex Value to hide=ScalarParameter() {defaultValue=0.3, group=ScareyBearyExpression, parameterName=R_squint_inner_pose_value}
  ) {materialFunction=&VertexColorMask}
  n7 = MaterialFunctionCall(
   Hide Element Toggle=@n2
   VertexColorChannel=@n4.R
   Vertex Value to hide=ScalarParameter() {defaultValue=0.2, group=ScareyBearyExpression, parameterName=C_glabella_up_pose_value}
  ) {materialFunction=&VertexColorMask}
  n8 = ScalarParameter() {defaultValue=0.1, group=ScareyBearyExpression, parameterName=default_pose_value}
  n9 = MaterialFunctionCall(Hide Element Toggle=@n2, VertexColorChannel=@n4.R, Vertex Value to hide=@n8) {materialFunction=&VertexColorMask}
  n10 = If(
   A=ScalarParameter() {group=ScareyBearyExpression, parameterName=C_glabella_up_pose, sortPriority=0}
   B=ScalarParameter() {defaultValue=0.1, group=ScareyBearyExpression, parameterName=Threshold_C_glabella_up_pose}
   "A > B"=@n7.Result
   "A == B"=@n9.Result
   "A < B"=@n9.Result
  )
  n11 = If(
   A=ScalarParameter() {group=ScareyBearyExpression, parameterName=R_squint_inner_pose, sortPriority=0}
   B=ScalarParameter() {defaultValue=0.1, group=ScareyBearyExpression, parameterName=Threshold_R_squint_inner_pose}
   "A > B"=@n6.Result
   "A == B"=@n10
   "A < B"=@n10
  )
  n12 = If(
   A=ScalarParameter() {group=ScareyBearyExpression, parameterName=R_blink_pose, sortPriority=0}
   B=ScalarParameter() {defaultValue=0.1, group=ScareyBearyExpression, parameterName=Threshold_R_blink_pose}
   "A > B"=@n5.Result
   "A == B"=@n11
   "A < B"=@n11
  )
  n13 = Max(A=Max(A=Max(A=@n9.Result, B=@n7.Result), B=@n6.Result), B=@n5.Result)
  n14 = Constant() {r=1}
  n15 = ChannelMaskParameter(None=@n3) {group=ScareyBearyExpression, maskChannel=Blue, parameterName=SBE_MouthChannel, sortPriority=-1, defaultValue=<rgba 0 0 1 0>}
  n16 = MaterialFunctionCall(
   Hide Element Toggle=@n14
   VertexColorChannel=@n15.R
   Vertex Value to hide=ScalarParameter() {defaultValue=0.4, group=ScareyBearyExpression, parameterName=R_Frown_pose_value}
  ) {materialFunction=&VertexColorMask}
  n17 = MaterialFunctionCall(
   Hide Element Toggle=@n14
   VertexColorChannel=@n15.R
   Vertex Value to hide=ScalarParameter() {defaultValue=0.3, group=ScareyBearyExpression, parameterName=R_smile_pose_value}
  ) {materialFunction=&VertexColorMask}
  n18 = MaterialFunctionCall(
   Hide Element Toggle=@n14
   VertexColorChannel=@n15.R
   Vertex Value to hide=ScalarParameter() {defaultValue=0.2, group=ScareyBearyExpression, parameterName=jaw_open_pose_value}
  ) {materialFunction=&VertexColorMask}
  n19 = MaterialFunctionCall(Hide Element Toggle=@n14, VertexColorChannel=@n15.R, Vertex Value to hide=@n8) {materialFunction=&VertexColorMask}
  n20 = If(
   A=ScalarParameter() {group=ScareyBearyExpression, parameterName=jaw_open_pose, sortPriority=0}
   B=ScalarParameter() {defaultValue=0.1, group=ScareyBearyExpression, parameterName=Threshold_jaw_open_pose}
   "A > B"=@n18.Result
   "A == B"=@n19.Result
   "A < B"=@n19.Result
  )
  n21 = If(
   A=ScalarParameter() {group=ScareyBearyExpression, parameterName=R_smile_pose, sortPriority=0}
   B=ScalarParameter() {defaultValue=1, group=ScareyBearyExpression, parameterName=Threshold_R_smile_pose}
   "A > B"=@n17.Result
   "A == B"=@n20
   "A < B"=@n20
  )
  n22 = If(
   A=ScalarParameter() {group=ScareyBearyExpression, parameterName=R_Frown_pose, sortPriority=0}
   B=ScalarParameter() {defaultValue=0.5, group=ScareyBearyExpression, parameterName=Threshold_R_Frown_pose}
   "A > B"=@n16.Result
   "A == B"=@n21
   "A < B"=@n21
  )
  n23 = Max(A=Max(A=Max(A=@n19.Result, B=@n18.Result), B=@n17.Result), B=@n16.Result)
  n24 = Add(A=Subtract(A=OneMinus(None=@n12), B=OneMinus(None=@n13)), B=Subtract(A=OneMinus(None=@n22), B=OneMinus(None=@n23)))
  n25 = Subtract(A=ObjectPositionWS(), B=WorldPosition().XYZ)
  FunctionOutput(None=SetMaterialAttributes!(MaterialAttributes=GetMaterialAttributes(None=@n1).MaterialAttributes, WorldPositionOffset=Multiply(A=@n24, B=@n25)) {attributeSetTypes [#1]})
 ]}
