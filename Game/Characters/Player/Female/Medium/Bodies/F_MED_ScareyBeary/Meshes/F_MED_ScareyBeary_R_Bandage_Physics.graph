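# Physics asset for the right-side bandage on the female medium ScareyBeary player body, giving the dangling strip collision and joint limits.
# A chain of small capsules runs from a base bone on the right forearm through five segments, joined with wide swing and tight twist limits.
# Larger capsules cover the torso, right arm, both thighs and right calf, linked by angularly free constraints, apparently as collision shapes for the bandage.
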PhysicsAsset /Game/Characters/Player/Female/Medium/Bodies/F_MED_ScareyBeary/Meshes/F_MED_ScareyBeary_R_Bandage_Physics
bodies {
 calf_r [{shapeType=Capsule, length=37.5, radius=10, shapeName=calf_r_capsule, center=<13.6903 0 0>, rotation=<rot -90 180 180>}]
 dyn_bandage_r_01 [{shapeType=Capsule, length=1, radius=1.5, shapeName=dyn_bandage_r_01_capsule, center=<4.2286 0 0>, rotation=<rot 90 0 0>}]
 dyn_bandage_r_02 [{shapeType=Capsule, length=1, radius=1.5, shapeName=dyn_bandage_r_02_capsule, center=<5.0276 0 0>, rotation=<rot 90 0 0>}]
 dyn_bandage_r_03 [{shapeType=Capsule, length=1, radius=2.5, shapeName=dyn_bandage_r_03_capsule, center=<3.4677 0 0>, rotation=<rot 90 0 0>}]
 dyn_bandage_r_04 [{shapeType=Capsule, length=2, radius=2.5, shapeName=dyn_bandage_r_04_capsule, center=<1.4207 0 0>, rotation=<rot 90 0 0>}]
 dyn_bandage_r_05 [{shapeType=Capsule, length=1, radius=2.5, shapeName=dyn_bandage_r_05_capsule, center=<1.9241 0 0>, rotation=<rot 90 0 0>}]
 dyn_bandage_r_base [{shapeType=Capsule, length=1, radius=1, shapeName=dyn_bandage_r_base_capsule}]
 hand_r [{shapeType=Capsule, length=5, radius=6.5, shapeName=hand_r_capsule, center=<-7.9384 0 0>, rotation=<rot -90 180 180>}]
 lowerarm_r [{shapeType=Capsule, length=11.5, radius=7, shapeName=lowerarm_r_capsule, center=<-9.002 0 -1.619>, rotation=<rot -90 180 180>}]
 pelvis [{shapeType=Capsule, length=12, radius=13.5, shapeName=pelvis_capsule}]
 spine_03 [{shapeType=Capsule, length=9, radius=11.5, shapeName=spine_03_capsule, center=<-3.3463 0 0>}]
 thigh_l [{shapeType=Capsule, length=27.8911, radius=9.6834, shapeName=thigh_l_capsule, center=<-22.4933 0.2753 0>, rotation=<rot 90 160.5288 160.529>}]
 thigh_r [{shapeType=Capsule, length=17, radius=11, shapeName=thigh_r_capsule, center=<20.4305 0 0>, rotation=<rot 90 0 0>}]
 upperarm_r [{shapeType=Capsule, length=13.5, radius=5.5, shapeName=upperarm_r_capsule, center=<-12.6156 0 0>, rotation=<rot -90 0 0>}]
}
constraints [
 {
  bone1Name=spine_03
  bone2Name=pelvis
  swing1LimitDegrees=45
  swing1Motion=Free
  swing2LimitDegrees=45
  swing2Motion=Free
  twistLimitDegrees=45
  twistMotion=Free
 }
 {
  bone1Name=upperarm_r
  bone2Name=spine_03
  swing1LimitDegrees=45
  swing1Motion=Free
  swing2LimitDegrees=45
  swing2Motion=Free
  twistLimitDegrees=45
  twistMotion=Free
 }
 {
  bone1Name=lowerarm_r
  bone2Name=upperarm_r
  swing1LimitDegrees=45
  swing1Motion=Free
  swing2LimitDegrees=45
  swing2Motion=Free
  twistLimitDegrees=45
  twistMotion=Free
 }
 {
  bone1Name=hand_r
  bone2Name=lowerarm_r
  swing1LimitDegrees=45
  swing1Motion=Free
  swing2LimitDegrees=45
  swing2Motion=Free
  twistLimitDegrees=45
  twistMotion=Free
 }
 {
  bone1Name=thigh_r
  bone2Name=pelvis
  swing1LimitDegrees=45
  swing1Motion=Free
  swing2LimitDegrees=45
  swing2Motion=Free
  twistLimitDegrees=45
  twistMotion=Free
 }
 {
  bone1Name=calf_r
  bone2Name=thigh_r
  swing1LimitDegrees=45
  swing1Motion=Free
  swing2LimitDegrees=45
  swing2Motion=Free
  twistLimitDegrees=45
  twistMotion=Free
 }
 {bone1Name=dyn_bandage_r_01, bone2Name=dyn_bandage_r_base, swing1LimitDegrees=100, swing2LimitDegrees=100, twistLimitDegrees=5}
 {bone1Name=dyn_bandage_r_02, bone2Name=dyn_bandage_r_01, swing1LimitDegrees=100, swing2LimitDegrees=100, twistLimitDegrees=5}
 {bone1Name=dyn_bandage_r_03, bone2Name=dyn_bandage_r_02, swing1LimitDegrees=100, swing2LimitDegrees=100, twistLimitDegrees=5}
 {bone1Name=dyn_bandage_r_04, bone2Name=dyn_bandage_r_03, swing1LimitDegrees=100, swing2LimitDegrees=100, twistLimitDegrees=5}
 {bone1Name=dyn_bandage_r_05, bone2Name=dyn_bandage_r_04, swing1LimitDegrees=100, swing2LimitDegrees=100, twistLimitDegrees=5}
 {
  bone1Name=dyn_bandage_r_base
  bone2Name=lowerarm_r
  swing1LimitDegrees=45
  swing1Motion=Free
  swing2LimitDegrees=45
  swing2Motion=Free
  twistLimitDegrees=45
  twistMotion=Free
 }
 {
  bone1Name=thigh_l
  bone2Name=pelvis
  swing1LimitDegrees=45
  swing1Motion=Free
  swing2LimitDegrees=45
  swing2Motion=Free
  twistLimitDegrees=45
  twistMotion=Free
 }
]
